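# Animation Blueprint for the DieselPunk pickaxe that adds procedural secondary motion to its dynamic bones.
# The output chain runs spring-driven Anim Dynamics on the hinge, handle and hose, with the handle at partial alpha and the hose allowed slight linear travel.
# Copy Bone Delta counter-rotates a second hinge bone from the simulated one.
# A similar chain for left and right port bones exists but is not connected to the output pose.
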AnimBlueprint /BRCosmetics/Weapons/FORT_Melee/Pickaxe_DieselPunk/Meshes/Pickaxe_DieselPunk_Skeleton_AnimBP
target_skeleton=&/Pickaxe_DieselPunk_Skeleton
graphs [
 AnimGraph {nodes [
   n1 = Anim Dynamics - Bone: dyn_portA1_l(Alpha=1, ExternalForce="0,0,0") {
    angularDampingOverride=0.999
    angularSpringConstant=50
    bAngularSpring=true
    bGravityOverrideInSimSpace=true
    bOverrideAngularDamping=true
    bUsePlanarLimit=false
    gravityScale=0.1
    alphaBoolBlend {blendOption=Linear}
    boundBone {boneName=dyn_portA1_l}
    componentAppliedLinearAccClamp=<1000 1000 1000>
    componentLinearAccScale=<1 1 1>
    componentLinearVelScale=<1 1 1>
    physicsBodyDefinitions [{
      sphereCollisionRadius=0
      boundBone {boneName=dyn_portA1_l}
      boxExtents=<8 8 2>
      constraintSetup {angularLimitsMax=<0 5 5>, angularLimitsMin=<0 -5 -5>, angularTarget=<0 0 0>}
      localJointOffset=<0 0 -6>
     }]
   }
   n2 = Copy Bone Delta - Source Bone: dyn_portA1_l - Target Bone: dyn_portA2_l!(ComponentPose:Component Space Pose Link Structure=@n1, Alpha:float=1) -> (Pose:Component Space Pose Link Structure) {
    is=title AnimGraphNode_CopyBoneDelta
    bCopyRotation=true
    copyMode=Copy
    alphaBoolBlend {blendOption=Linear}
    sourceBone {boneName=dyn_portA1_l}
    targetBone {boneName=dyn_portA2_l}
   }
   n3 = Copy Bone Delta - Source Bone: dyn_portA1_l - Target Bone: dyn_portA2_l(ComponentPose=@n2, Alpha=1) {
    bCopyRotation=true
    copyMode=Copy
    alphaBoolBlend {blendOption=Linear}
    sourceBone {boneName=dyn_portA1_l}
    targetBone {boneName=dyn_portA2_l}
   }
   n4 = Anim Dynamics - Bone: dyn_portA1_r(ComponentPose=@n3, Alpha=1, ExternalForce="0,0,0") {
    angularDampingOverride=0.999
    angularSpringConstant=50
    bAngularSpring=true
    bGravityOverrideInSimSpace=true
    bOverrideAngularDamping=true
    bUsePlanarLimit=false
    gravityScale=0.1
    alphaBoolBlend {blendOption=Linear}
    boundBone {boneName=dyn_portA1_r}
    componentAppliedLinearAccClamp=<1000 1000 1000>
    componentLinearAccScale=<1 1 1>
    componentLinearVelScale=<1 1 1>
    physicsBodyDefinitions [{
      sphereCollisionRadius=0
      boundBone {boneName=dyn_portA1_r}
      boxExtents=<8 8 2>
      constraintSetup {angularLimitsMax=<0 5 5>, angularLimitsMin=<0 -5 -5>, angularTarget=<0 0 0>}
      localJointOffset=<0 0 -6>
     }]
   }
   n5 = Copy Bone Delta - Source Bone: dyn_portA1_r - Target Bone: dyn_portA2_r!(ComponentPose:Component Space Pose Link Structure=@n4, Alpha:float=1) -> (Pose:Component Space Pose Link Structure) {
    is=title AnimGraphNode_CopyBoneDelta
    bCopyRotation=true
    copyMode=Copy
    alphaBoolBlend {blendOption=Linear}
    sourceBone {boneName=dyn_portA1_r}
    targetBone {boneName=dyn_portA2_r}
   }
   Copy Bone Delta - Source Bone: dyn_portA1_r - Target Bone: dyn_portA3_r!(ComponentPose:Component Space Pose Link Structure=@n5, Alpha:float=1) -> (Pose:Component Space Pose Link Structure) {
    is=title AnimGraphNode_CopyBoneDelta
    bCopyRotation=true
    copyMode=Copy
    alphaBoolBlend {blendOption=Linear}
    sourceBone {boneName=dyn_portA1_r}
    targetBone {boneName=dyn_portA3_r}
   }
   n6 = Anim Dynamics - Bone: dyn_hingeA(ComponentPose=Local To Component(), Alpha=1, ExternalForce="0,0,0") {
    angularDampingOverride=0.9999
    angularSpringConstant=2500
    bAngularSpring=true
    bGravityOverrideInSimSpace=true
    bOverrideAngularDamping=true
    bUsePlanarLimit=false
    gravityScale=0.1
    alphaBoolBlend {blendOption=Linear}
    boundBone {boneName=dyn_hingeA}
    componentAppliedLinearAccClamp=<1000 1000 1000>
    componentLinearAccScale=<1 1 1>
    componentLinearVelScale=<1 1 1>
    physicsBodyDefinitions [{
      sphereCollisionRadius=0
      boundBone {boneName=dyn_hingeA}
      boxExtents=<22 22 8>
      constraintSetup {twistAxis=AxisY, angularLimitsMax=<5 0 0>, angularLimitsMin=<-10 0 0>, angularTarget=<0 0 0>}
      localJointOffset=<0 -9 -9>
     }]
   }
   n7 = Copy Bone Delta - Source Bone: dyn_hingeA - Target Bone: dyn_hingeA1!(ComponentPose:Component Space Pose Link Structure=@n6, Alpha:float=1) -> (Pose:Component Space Pose Link Structure) {
    is=title AnimGraphNode_CopyBoneDelta
    bCopyRotation=true
    copyMode=Copy
    rotationMultiplier=-1
    alphaBoolBlend {blendOption=Linear}
    sourceBone {boneName=dyn_hingeA}
    targetBone {boneName=dyn_hingeA1}
   }
   n8 = Anim Dynamics - Bone: dyn_handle(ComponentPose=@n7, Alpha=0.2, ExternalForce="0,0,0") {
    angularDampingOverride=0.9999
    angularSpringConstant=1000
    bAngularSpring=true
    bGravityOverrideInSimSpace=true
    bOverrideAngularDamping=true
    bUsePlanarLimit=false
    gravityScale=0.1
    alphaBoolBlend {blendOption=Linear}
    boundBone {boneName=dyn_handle}
    componentAppliedLinearAccClamp=<1000 1000 1000>
    componentLinearAccScale=<1 1 1>
    componentLinearVelScale=<1 1 1>
    physicsBodyDefinitions [{
      sphereCollisionRadius=0
      boundBone {boneName=dyn_handle}
      boxExtents=<12 12 4>
      constraintSetup {twistAxis=AxisY, angularLimitsMax=<5 0 0>, angularLimitsMin=<-5 0 0>, angularTarget=<0 0 0>}
      localJointOffset=<0 -1 -9>
     }]
   }
   n9 = Anim Dynamics - Bone: dyn_hose(ComponentPose=@n8, Alpha=1, ExternalForce="0,0,0") {
    angularDampingOverride=0.9999
    angularSpringConstant=1000
    bAngularSpring=true
    bGravityOverrideInSimSpace=true
    bOverrideAngularDamping=true
    bUsePlanarLimit=false
    gravityScale=0.1
    alphaBoolBlend {blendOption=Linear}
    boundBone {boneName=dyn_hose}
    componentAppliedLinearAccClamp=<1000 1000 1000>
    componentLinearAccScale=<1 1 1>
    componentLinearVelScale=<1 1 1>
    physicsBodyDefinitions [{
      sphereCollisionRadius=0
      boundBone {boneName=dyn_hose}
      boxExtents=<12 12 4>
      constraintSetup {twistAxis=AxisY, angularLimitsMax=<0 0 0>, angularLimitsMin=<0 0 0>, angularTarget=<0 0 0>, linearAxesMax=<0 0 1>, linearAxesMin=<0 0 -1>}
      localJointOffset=<0 0 0>
     }]
   }
   Output Pose(Result=Component To Local(ComponentPose=@n9)) {layerGroup=None}
  ]}
 EventGraph {nodes [Event BlueprintUpdateAnimation()]}
]
variables []
