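# Idle Niagara effect for the DieselPunk pickaxe, adding exhaust smoke and a flickering thruster flame.
# Exhausts spawns dark, growing, fading sub-UV smoke sprites, split by partition across three offset points.
# Thruster spawns hot orange flame sprites, while a waveform flicker value drives emitter-level glow color and size, with light color derived from the glow for its light renderer.
# It exposes a User.bisSwinging bool, though the emitter modules shown never read it.
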NiagaraSystem /BRCosmetics/Weapons/FORT_Melee/Pickaxe_DieselPunk/NS_DieselPunk_Pickaxe_Idle
dependencies {
 usedDataInterfaces [
  &NiagaraDataInterfaceCurve
  &NiagaraDataInterfaceVectorCurve
  &NiagaraDataInterfaceSpriteRendererInfo
 ]
 usedDynamicInputs [&Boolean_NotOperation]
 usedModules [
  &SystemState
  &self:SystemSpawnScript.SystemScriptSource.SystemScriptGraph.NiagaraNodeAssignment_0.SetVariables_#1
 ]
}
emitters {
 Exhausts {
  emitterUpdateScript [
   &EmitterState()
   &SpawnRate(SpawnRate=20, Use Spawn Probability=false)
  ]
  particleSpawnScript [
   &PartitionParticles(Number Of Partitions=3)
   &InitializeParticle(
    Color=<rgba 0.25 0.25 0.25 1>
    Lifetime=&RandomRangeFloat(Minimum=0.5, Maximum=0.8)
    Position Offset=&AddVectorToPosition(
     Position=<17 -1 35>
     Vector=&MergeHelper_VectorSwitch(0=<0 10 0>, 1=<0 4 0>, 2=<0 -3 0>, 3=?, Index=$Particles.PartitionParticles.Partition, Num=::NewEnumerator1)
    )
    Sprite Rotation Mode=::NewEnumerator2
    Sprite Size Mode=::NewEnumerator3
    Uniform Sprite Size=&RandomRangeFloat(Minimum=8, Maximum=10)
    UsePositionOffset=true
    Write Material Random=true
   )
   &AddVelocity(Velocity=&RandomRangeVector(Minimum=<5 -15 -3>, Maximum=<-5 -15 -5>, Evaluation Type=::NewEnumerator1))
   &CameraOffset(Camera Offset Amount=5)
  ]
  particleUpdateScript [
   &ParticleState()
   &ScaleSpriteSize(Uniform Curve Sprite Scale=DataInterface() {propertyValues {Curve {keys [{arriveTangent=5.2822, interpMode=RCIM_Cubic, leaveTangent=5.2822, tangentMode=RCTM_User, time=0, value=0.6}, <at 1 2.5>]}}})
   &SolveForcesAndVelocity(Clamp Velocity=false, Limit Acceleration=false)
   &ScaleColor(
    Scale Alpha=&FloatFromCurve(
     FloatCurve=DataInterface() {propertyValues {Curve {keys [<at 0 -0.0004 RCIM_Cubic>, {arriveTangent=0.0004, interpMode=RCIM_Cubic, leaveTangent=0.0004, time=0.22, value=1}, <at 1 0>]}}}
     Scale Curve=0.8
    )
    Scale RGB=&VectorFromCurve(VectorCurve=DataInterface() {propertyValues {
       XCurve {keys [{interpMode=RCIM_Cubic, tangentMode=RCTM_User, time=0, value=0.15}, <at 0.25 0.25 RCIM_Cubic>]}
       YCurve {keys [{interpMode=RCIM_Cubic, tangentMode=RCTM_User, time=0, value=0.15}, <at 0.25 0.25 RCIM_Cubic>]}
       ZCurve {keys [{interpMode=RCIM_Cubic, tangentMode=RCTM_User, time=0, value=0.15}, <at 0.25 0.25 RCIM_Cubic>]}
      }})
    ScaleA=true
    ScaleRGB=true
   )
   &DynamicMaterialParameters(
    Index 0 Param 1=&FloatFromCurve(FloatCurve=DataInterface() {propertyValues {Curve {keys [<at 0 1 RCIM_Cubic>,<at 1 -1>]}}})
    Index 0 Param 2=1
    Index 0 Param 3=1
    Index 0 Param 4=1
    Module.Index0Float0Write=true
    Module.Index0Float1Write=true
    Module.Index0Float2Write=true
    Module.Index0Float3Write=true
   )
   &SubUVAnimation(
    Frames Per Second=10
    Playback Mode:ENiagara_FPSPlayrate=::NewEnumerator1
    "Sprite Renderer "=DataInterface() {propertyValues {SpriteRenderer=&self:Exhausts_0.NiagaraSpriteRendererProperties_0}}
    SubUV Animation Mode:ENiagara_SubUVLookupModeV2=::NewEnumerator3
    UseEndFrame=false
    UsePlayRateScale=false
    UseStartFrame=false
   )
  ]
  rendererClasses [&NiagaraSpriteRendererProperties]
  renderers [{rendererClass=&NiagaraSpriteRendererProperties}]
 }
 Thruster {
  emitterUpdateScript [
   &EmitterState()
   &SpawnRate(SpawnRate=3, Use Spawn Probability=false)
   /SetVariables(Emitter.ThrusterFlickerAlpha:NiagaraFloat=&Waveform("Waveform [1]"=::NewEnumerator12, "[1] Frequency"=16, "[1] Amplitude Min/Max"=<0 1>)) {isSetParameters=true}
   /SetVariables(
    Emitter.GlowColor:LinearColor=&MultiplyLinearColor(
     Linear Color=<rgba 15 10 1 1>
     Scale Factor=&MakeVector4FromVectorAndFloat(
      "Vector (RGB)"=<1 0.25 0>
      "Float (Alpha)"=&FloatFromCurve(
       FloatCurve=DataInterface() {propertyValues {Curve {keys [<at 0 0>,<at 0.5 1>]}}}
       CurveIndex=$System.Age
       Scale Curve=&Lerp_Float(A=0.5, B=1, Alpha=$Emitter.ThrusterFlickerAlpha)
      )
     )
    )
    Emitter.GlowSize:Vector2f=&Vector2DFromFloat(Value=&Multiply_Float(A=8.5, B=&Lerp_Float(A=0.8, B=1, Alpha=$Emitter.ThrusterFlickerAlpha)))
    Emitter.PositionGlow:NiagaraPosition=<1 -30 48>
   ) {isSetParameters=true}
   /SetVariables(Emitter.CameraOffset=2.5) {isSetParameters=true}
   /SetVariables(
    Emitter.LightColor:LinearColor=&MultiplyLinearColor(Linear Color=$Emitter.GlowColor, Scale Factor=<1 1 1 10>)
    Emitter.PositionLight:NiagaraPosition=&AddVectorToPosition(Position=$Emitter.PositionGlow, Vector=<0 -7 0>)
   ) {isSetParameters=true}
  ]
  particleSpawnScript [
   &InitializeParticle(
    Color=<rgba 1 0.6 0.02 1>
    Lifetime=0.6
    Position Offset=&AddVectorToPosition(Position=$Emitter.PositionGlow, Vector=<0 -1 0>)
    Sprite Rotation Mode=::NewEnumerator2
    Sprite Size Mode=::NewEnumerator3
    Uniform Sprite Size=&RandomRangeFloat(Minimum=6, Maximum=8, Evaluation Type=::NewEnumerator1)
    UsePositionOffset=true
    Write Material Random=true
   )
   &AddVelocity(Velocity=<0 -5 0>)
  ]
  particleUpdateScript [
   &ParticleState()
   &DynamicMaterialParameters(
    Index 0 Param 1=&FloatFromCurve(FloatCurve=DataInterface() {propertyValues {Curve {keys [<at 0.5 1 RCIM_Cubic>,<at 1 -1>]}}})
    Index 0 Param 2=1
    Index 0 Param 3=1
    Index 0 Param 4=1
    Module.Param0WriteEnabled=true
    Module.Param1WriteEnabled=true
    Module.Param2WriteEnabled=true
    Module.Param3WriteEnabled=true
   )
   &SubUVAnimation(
    Frames Per Second=10
    Playback Mode:ENiagara_FPSPlayrate=::NewEnumerator1
    Playback Scale Curve=DataInterface() {propertyValues {Curve {keys [
        {interpMode=RCIM_Cubic, tangentMode=RCTM_User, time=0.25, value=1}
        {arriveTangent=-1.2656, interpMode=RCIM_Cubic, leaveTangent=-1.2656, tangentMode=RCTM_User, time=1, value=0}
       ]}}}
    Random Start Frame=true
    "Sprite Renderer "=DataInterface() {propertyValues {SpriteRenderer=&self:Thruster_0.NiagaraSpriteRendererProperties_1}}
    SubUV Animation Mode:ENiagara_SubUVLookupModeV2=::NewEnumerator3
    UseEndFrame=false
    UsePlayRateScale=false
    UseStartFrame=false
   )
   &ScaleColor(
    Scale Alpha=&FloatFromCurve(FloatCurve=DataInterface() {propertyValues {Curve {keys [
         {arriveTangent=4.692, interpMode=RCIM_Cubic, leaveTangent=4.692, tangentMode=RCTM_User, time=0, value=0}
         {arriveTangent=-0.059, interpMode=RCIM_Cubic, leaveTangent=-0.059, tangentMode=RCTM_User, time=0.1, value=1}
         {arriveTangent=-0.0141, interpMode=RCIM_Cubic, leaveTangent=-0.0141, tangentMode=RCTM_User, time=0.5987, value=0.9935}
         {arriveTangent=-1.7744, tangentMode=RCTM_User, time=1, value=0}
        ]}}})
    Scale RGB=&VectorFromCurve(VectorCurve=DataInterface() {propertyValues {
       XCurve {keys [{arriveTangent=47.7435, interpMode=RCIM_Cubic, leaveTangent=47.7435, tangentMode=RCTM_User, time=0, value=15}, <at 0.2668 25 RCIM_Cubic>]}
       YCurve {keys [{arriveTangent=16.22, interpMode=RCIM_Cubic, leaveTangent=16.22, tangentMode=RCTM_User, time=0, value=0.25}, <at 0.2668 5 RCIM_Cubic>]}
       ZCurve {keys [{arriveTangent=-0.0001, interpMode=RCIM_Cubic, leaveTangent=-0.0001, tangentMode=RCTM_User, time=0, value=0.1}, <at 0.2668 1 RCIM_Cubic>]}
      }})
    ScaleA=true
    ScaleRGB=true
   )
   &ScaleSpriteSize(
    Scale Sprite Size Mode=::NewEnumerator0
    Uniform Scale Factor=&FloatFromCurve(FloatCurve=DataInterface() {propertyValues {Curve {keys [{arriveTangent=1.4511, interpMode=RCIM_Cubic, leaveTangent=1.4511, tangentMode=RCTM_User, time=0, value=0.5}, <at 1 1.25>]}}})
   )
   &SolveForcesAndVelocity(Clamp Velocity=false, Limit Acceleration=false)
   &CameraOffset(Camera Offset Amount=2)
  ]
  rendererClasses [&NiagaraSpriteRendererProperties, &NiagaraLightRendererProperties]
  renderers [
   {rendererClass=&NiagaraSpriteRendererProperties}
   {rendererClass=&NiagaraLightRendererProperties, rendererIndex=1}
   {rendererClass=&NiagaraSpriteRendererProperties, rendererIndex=2}
  ]
 }
}
summary {
 systemName=NS_DieselPunk_Pickaxe_Idle
 emitters [
  {emitterName=Exhausts, rendererClasses [&NiagaraSpriteRendererProperties]}
  {emitterName=Thruster, rendererClasses [&NiagaraSpriteRendererProperties, &NiagaraLightRendererProperties]}
 ]
 userVariables [User.bisSwinging {type=NiagaraBool}]
}
