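# Niagara swing effect for the DieselPunk pickaxe: a brief thruster burst with a flame trail, smoke and embers.
# Thruster, Embers and Smoke sample a shared socket reader; embers and smoke inherit socket velocity, embers adding curl noise and delayed gravity.
# Thruster drives mesh scale, color and light from emitter-age curves alongside short-lived sprites.
# TrailFlame builds an anim-trail with two ribbon renderers, carrying separate smoke color and width, faded by a system-level value.
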NiagaraSystem /BRCosmetics/Weapons/FORT_Melee/Pickaxe_DieselPunk/NS_DieselPunk_Pickaxe_Swing
dependencies {
 usedDataInterfaces [
  &NiagaraDataInterfaceCurve
  &NiagaraDataInterfaceColorCurve
  &NiagaraDataInterfaceVectorCurve
  &NiagaraDataInterfaceVector2DCurve
  &NiagaraDataInterfaceSpriteRendererInfo
  &NiagaraDataInterfaceSocketReader
 ]
 usedDynamicInputs [&RandomRangeVector]
 usedModules [
  &SystemState
  &self:SystemSpawnScript.SystemScriptSource.SystemScriptGraph.NiagaraNodeAssignment_0.SetVariables_#1
  &self:SystemSpawnScript.SystemScriptSource.SystemScriptGraph.NiagaraNodeAssignment_1.SetVariables_#2
 ]
}
emitters {
 Embers {
  emitterUpdateScript [
   &EmitterState(Life Cycle Mode=::NewEnumerator1, Loop Behavior=::NewEnumerator1, Loop Delay=0.035, Loop Duration=0.17, UseLoopDelay=true)
   &SpawnPerUnit(Module.Use Max Movement Threshold=true, Module.Use Movement Tolerance=false, Spawn Spacing=30, Use Spawn Probability=false)
   &SpawnRate(SpawnRate=70, Use Spawn Probability=true)
  ]
  particleSpawnScript [
   &PartitionParticles()
   &InitializeParticle(
    Color=<rgba 4 3.8667 3.6 1>
    Lifetime=&RandomRangeFloat(Minimum=0.2, Maximum=0.4)
    Sprite Rotation Mode=::NewEnumerator2
    Sprite Size Max=<1.5 4>
    Sprite Size Min=<0.7 2>
    Sprite Size Mode=::NewEnumerator2
    UsePositionOffset=false
    Write Material Random=true
   )
   &SocketLocation(
    Interpolate Socket Sampling=true
    SamplePosOffset=false
    Sampled Position Offset=<0 -20 0>
    Socket Reader=$System.Socket Reader
    Socket Sampling Mode=::NewEnumerator3
    VelocitySpeedLimit=false
   )
   &ShapeLocation(
    MaskSpawnGroup=false
    Offset Mode=::NewEnumerator0
    Override Seed=false
    Rotation Axis=<0 1 0>
    Rotation Mode=::NewEnumerator1
    Sphere Radius=15
    Sphere Surface Distribution=0.0476
    UseOwnerScale=true
   )
   &AddVelocity(
    Module.Use Falloff Distance=false
    Override Seed=false
    Velocity Mode=::NewEnumerator1
    Velocity Origin=$Output.SocketLocation.SampledPosition
    Velocity Speed=&RandomRangeFloat(Minimum=20, Maximum=80)
   )
   &InheritVelocity(
    Inherited Velocity Speed Limit:NiagaraFloat=&RandomRangeFloat(Minimum=60, Maximum=300)
    Speed Limit=true
    UseSpeedThreshold=false
    Velocity Source=$Output.SocketLocation.SampledVelocity
   )
  ]
  particleUpdateScript [
   &ParticleState()
   &ScaleSpriteSize(
    Uniform Curve Scale=&RandomRangeFloat(Minimum=0.8, Maximum=1.7)
    Uniform Curve Sprite Scale=DataInterface() {propertyValues {Curve {keys [
        <at 0 0 RCIM_Cubic>
        <at 0.2 1 RCIM_Cubic>
        {arriveTangent=-2.2922, interpMode=RCIM_Cubic, leaveTangent=-2.2922, tangentMode=RCTM_User, time=1, value=0}
       ]}}}
   )
   &Drag(Drag=&RandomRangeFloat(Minimum=2, Maximum=4.5), Use Linear Drag=true, Use Rotational Drag=true)
   &ScaleColor(
    Scale Alpha=&FloatFromCurve(FloatCurve=DataInterface() {propertyValues {Curve {keys [
         {interpMode=RCIM_Cubic, tangentMode=RCTM_User, time=0, value=0}
         {arriveTangent=-0.625, interpMode=RCIM_Cubic, leaveTangent=-0.625, tangentMode=RCTM_User, time=0.2, value=1}
        ]}}})
    Scale RGB=&MakeVectorFromLinearColorRGB(Color=&ColorFromCurve(ColorCurve=DataInterface() {propertyValues {
        AlphaCurve {keys [<at 0 1 RCIM_Cubic>]}
        BlueCurve {keys [<at 0 0.04 RCIM_Cubic>,<at 0.331 0.0192>,<at 0.9443 0>]}
        GreenCurve {keys [<at 0 0.7587 RCIM_Cubic>,<at 0.331 0.3688>,<at 0.9443 0.02>]}
        RedCurve {keys [<at 0 2 RCIM_Cubic>,<at 0.331 1>,<at 0.9443 0.2>]}
       }}))
    ScaleA=true
    ScaleRGB=true
   )
   /SetVariables(Particles.Glow03_Size=&Multiply_Vector2D(A=$Particles.SpriteSize, B=<4 3>)) {isSetParameters=true}
   /SetVariables(Particles.Glow03_Color=&MakeLinearColorFromVectorAndFloat("Vector (RGB)"=&MakeVectorFromLinearColorRGB(Color=<rgba 1 0.37 0.1 1>), "Float (Alpha)"=0.33)) {isSetParameters=true}
   &GravityForce(Gravity=&Multiply_VectorByFloat(
     Vector=<0 0 -980>
     Float=&FloatFromCurve(FloatCurve=DataInterface() {propertyValues {Curve {keys [<at 0.4 0>,<at 1 1>]}}})
    ))
   &CurlNoiseForce(
    Noise Frequency=&RandomRangeFloat(Minimum=60, Maximum=100)
    Noise Strength=&Multiply_Float(
     A=&RandomRangeFloat(Minimum=400, Maximum=600)
     B=&FloatFromCurve(FloatCurve=DataInterface() {propertyValues {Curve {keys [<at 0 1>,<at 1 0>]}}})
    )
    Pan Noise=true
    Pan Noise Field=<0.2 0 0.3>
   )
   &SolveForcesAndVelocity(Clamp Velocity=false, Limit Acceleration=false)
   &ScaleSpriteSizeBySpeed(Max Scale Factor=<0.8 1.3>, Min Scale Factor=<1 0.2>, Velocity Threshold=100)
  ]
  rendererClasses [&NiagaraSpriteRendererProperties]
  renderers [
   {rendererClass=&NiagaraSpriteRendererProperties}
   {rendererClass=&NiagaraSpriteRendererProperties, rendererIndex=1}
  ]
 }
 Smoke {
  emitterUpdateScript [
   &EmitterState(Life Cycle Mode=::NewEnumerator1, Loop Behavior=::NewEnumerator1, Loop Delay=0.04, Loop Duration=0.16, UseLoopDelay=true)
   &SpawnPerUnit(Module.Use Max Movement Threshold=true, Module.Use Movement Tolerance=false, Use Spawn Probability=false) {enabled=false}
   &SpawnRate(SpawnRate=110, Use Spawn Probability=true)
  ]
  particleSpawnScript [
   &PartitionParticles()
   &InitializeParticle(
    Color=&MultiplyLinearColor(
     Linear Color=<rgba 0.125 0.121 0.11 1>
     Scale Factor=&MakeVector4FromVectorAndFloat(
      "Vector (RGB)"=<1 1 1>
      "Float (Alpha)"=&FloatFromCurve(
       FloatCurve=DataInterface() {propertyValues {Curve {keys [
           {interpMode=RCIM_Cubic, tangentMode=RCTM_User, time=0, value=1}
           {arriveTangent=-3.661, interpMode=RCIM_Cubic, leaveTangent=-3.661, tangentMode=RCTM_User, time=0.3, value=0.3}
          ]}}}
       CurveIndex=$Emitter.Age
      )
     )
    )
    Lifetime=&RandomRangeFloat(Minimum=0.22, Maximum=0.33)
    Sprite Rotation Mode=::NewEnumerator2
    Sprite Size Mode=::NewEnumerator3
    Uniform Sprite Size=&Multiply_Float(A=&RandomRangeFloat(Minimum=60, Maximum=70), B=1)
    UsePositionOffset=false
    Write Material Random=true
   )
   &SocketLocation(
    Interpolate Socket Sampling=true
    SamplePosOffset=false
    Sampled Position Offset=<0 -20 0>
    Socket Reader=$System.Socket Reader
    Socket Sampling Mode=::NewEnumerator3
    VelocitySpeedLimit=false
   )
   &ShapeLocation(
    MaskSpawnGroup=false
    Offset Mode=::NewEnumerator0
    Override Seed=false
    Rotation Axis=<0 1 0>
    Rotation Mode=::NewEnumerator1
    Sphere Radius=5
    Sphere Surface Distribution=0.1905
    UseOwnerScale=true
   )
   &InheritVelocity(
    Inherited Velocity Amount Scale=<0.1 0.1 0>
    Inherited Velocity Speed Limit:NiagaraFloat=&RandomRangeFloat(Minimum=60, Maximum=100)
    Speed Limit=true
    UseSpeedThreshold=false
    Velocity Source=$Output.SocketLocation.SampledVelocity
   )
  ]
  particleUpdateScript [
   &ParticleState()
   &ScaleSpriteSize(Uniform Curve Sprite Scale=DataInterface() {propertyValues {Curve {keys [{arriveTangent=1.1948, interpMode=RCIM_Cubic, leaveTangent=1.1948, tangentMode=RCTM_User, time=0, value=0.33}, <at 1 1 RCIM_Cubic>]}}})
   &Drag(Drag=6, Use Linear Drag=true, Use Rotational Drag=true)
   &ScaleColor(
    Scale Alpha=&FloatFromCurve(
     FloatCurve=DataInterface() {propertyValues {Curve {keys [
         {arriveTangent=14.9796, interpMode=RCIM_Cubic, leaveTangent=14.9796, tangentMode=RCTM_User, time=0, value=0}
         {arriveTangent=-0.625, interpMode=RCIM_Cubic, leaveTangent=-0.625, tangentMode=RCTM_User, time=0.2, value=1}
         {arriveTangent=-1.7041, interpMode=RCIM_Cubic, leaveTangent=-1.7041, tangentMode=RCTM_User, time=1, value=0}
        ]}}}
     Scale Curve=&RandomRangeFloat(Minimum=0.33, Maximum=0.5)
    )
    Scale RGB=&MakeVectorFromLinearColorRGB(Color=&ColorFromCurve(ColorCurve=DataInterface() {propertyValues {
        AlphaCurve {keys [<at 0 1 RCIM_Cubic>]}
        BlueCurve {keys [<at 0.0916 0.72 RCIM_Cubic>,<at 0.4219 1>]}
        GreenCurve {keys [<at 0.0916 2.216 RCIM_Cubic>,<at 0.4219 1>]}
        RedCurve {keys [<at 0.0916 6 RCIM_Cubic>,<at 0.4219 1>]}
       }}))
    ScaleA=true
    ScaleRGB=true
   )
   &AccelerationForce(Acceleration=&Multiply_VectorByFloat(Vector=<0 0 500>, Float=&RandomRangeFloat(Minimum=0.3, Maximum=0.5)))
   &SolveForcesAndVelocity(Clamp Velocity=false, Limit Acceleration=false)
   &SubUVAnimation(
    Frames Per Second=25
    Playback Mode:ENiagara_FPSPlayrate=::NewEnumerator1
    "Sprite Renderer "=DataInterface() {propertyValues {SpriteRenderer=&self:Smoke_1.NiagaraSpriteRendererProperties_1}}
    SubUV Animation Mode:ENiagara_SubUVLookupModeV2=::NewEnumerator3
    UseEndFrame=false
    UsePlayRateScale=false
    UseStartFrame=false
   )
   &DynamicMaterialParameters(
    Index 0 Param 1=&FloatFromCurve(FloatCurve=DataInterface() {propertyValues {Curve {keys [<at 0.5 1 RCIM_Cubic>,<at 0.9984 -1.0007 RCIM_Cubic>]}}})
    Index 0 Param 2=1
    Index 0 Param 3=1
    Index 0 Param 4=0.5
    Module.Index0Float0Write=true
    Module.Index0Float1Write=true
    Module.Index0Float2Write=true
    Module.Index0Float3Write=true
   )
  ]
  rendererClasses [&NiagaraSpriteRendererProperties]
  renderers [{rendererClass=&NiagaraSpriteRendererProperties}]
 }
 Thruster {
  emitterUpdateScript [
   &EmitterState(Life Cycle Mode=::NewEnumerator1, Loop Behavior=::NewEnumerator1, Loop Duration=0.1, UseLoopDelay=false)
   /SetVariables(Emitter.ThrusterAge:NiagaraFloat=&FloatFromCurve(FloatCurve=DataInterface() {propertyValues {Curve {keys [<at 0 0>,<at 0.16 1>]}}}, CurveIndex=$Emitter.Age)) {isSetParameters=true}
   &SpawnRate(SpawnRate=70, Use Spawn Probability=false)
   /SetVariables(Emitter.Position:NiagaraPosition=<0 -32 50>) {isSetParameters=true}
   /SetVariables(
    Emitter.Color:LinearColor=&MultiplyLinearColor(
     Linear Color=<rgba 50 12.875 0.5 25>
     Scale Factor=&MakeVector4FromVectorAndFloat(
      "Vector (RGB)"=&VectorFromCurve(
       VectorCurve=DataInterface() {propertyValues {XCurve {keys [<at 0 2>,<at 1 3>]}, YCurve {keys [<at 0 0.5>,<at 1 0.5>]}, ZCurve {keys [<at 0 0.25>,<at 1 0.2>]}}}
       CurveIndex=$Emitter.ThrusterAge
      )
      "Float (Alpha)"=&FloatFromCurve(
       FloatCurve=DataInterface() {propertyValues {Curve {keys [
           <at 0 0 RCIM_Cubic>
           {arriveTangent=5, interpMode=RCIM_Cubic, leaveTangent=5, time=0.0831, value=0.9348}
           {arriveTangent=-1.304, interpMode=RCIM_Cubic, leaveTangent=-1.304, time=0.2, value=1}
           <at 0.8 0 RCIM_Cubic>
          ]}}}
       CurveIndex=$Emitter.ThrusterAge
      )
     )
    )
    Emitter.MeshOrientation:Quat4f=<1 0 0 1>
    Emitter.Scale:Vector3f=&Multiply_Vector(
     A=<0.175 0.175 0.15>
     B=&VectorFromCurve(
      VectorCurve=DataInterface() {propertyValues {
        XCurve {keys [{interpMode=RCIM_Cubic, tangentMode=RCTM_User, time=0, value=1}, <at 0.25 1 RCIM_Cubic>, <at 0.2086 1 RCIM_Cubic>, <at 1 1 RCIM_Cubic>]}
        YCurve {keys [{interpMode=RCIM_Cubic, tangentMode=RCTM_User, time=0, value=1}, <at 0.25 1 RCIM_Cubic>, <at 0.2086 1 RCIM_Cubic>, <at 1 1 RCIM_Cubic>]}
        ZCurve {keys [
          {arriveTangent=17.5542, interpMode=RCIM_Cubic, leaveTangent=17.5542, tangentMode=RCTM_User, time=0, value=0}
          {arriveTangent=8.937, interpMode=RCIM_Cubic, leaveTangent=8.937, tangentMode=RCTM_User, time=0.25, value=0.25}
          {arriveTangent=-0.041, interpMode=RCIM_Cubic, leaveTangent=-0.041, tangentMode=RCTM_User, time=0.2086, value=2}
          <at 1 0 RCIM_Cubic>
         ]}
       }}
      CurveIndex=$Emitter.ThrusterAge
     )
    )
   ) {isSetParameters=true}
   /SetVariables(Emitter.DynParFlame:Vector4f=<0.4 1 1 1>) {isSetParameters=true}
   /SetVariables(Emitter.LightColor:LinearColor=&ColorFromCurve(
     ColorCurve=DataInterface() {propertyValues {
       AlphaCurve {keys [<at 0 0>, {arriveTangent=3.0215, interpMode=RCIM_Cubic, leaveTangent=3.0215, time=0.0613, value=1}, <at 0.331 1>, <at 0.98 0 RCIM_Cubic>]}
       BlueCurve {keys [<at 0.0381 0.5 RCIM_Cubic>,<at 0.2768 0.2>,<at 0.9477 0.1>]}
       GreenCurve {keys [<at 0.0381 14.525 RCIM_Cubic>,<at 0.2768 5.7716>,<at 0.9477 1.915>]}
       RedCurve {keys [<at 0.0381 50 RCIM_Cubic>,<at 0.2768 20>,<at 0.9477 10>]}
      }}
     CurveIndex=$Emitter.ThrusterAge
    )) {isSetParameters=true}
  ]
  particleSpawnScript [
   &InitializeParticle(
    Color=<rgba 5 1.925 0.9 1>
    Lifetime=0.06
    Sprite Size=<10 20>
    Sprite Size Mode=::NewEnumerator4
    UsePositionOffset=false
    Write Material Random=true
   )
   &SocketLocation(SamplePosOffset=false, Socket Reader=$System.Socket Reader, Socket Sampling Mode=::NewEnumerator3, VelocitySpeedLimit=false)
   &SpriteFacingAndAlignment(Alignment Coordinate Space=::Local, WriteAlignment=true, WriteFacing=false)
   &AddVelocity(Velocity=<0 300 0>, Velocity Speed Scale=-1)
  ]
  particleUpdateScript [
   &ParticleState()
   &ScaleSpriteSize(
    Scale Factor=&Vector2DFromCurve(Vector2Curve=DataInterface() {propertyValues {XCurve {keys [<at 0 1.5>,<at 1 0.25>]}, YCurve {keys [<at 0 1>,<at 1 4>]}}}, ScaleCurve=<1 1>)
    Scale Sprite Size Mode=::NewEnumerator2
   )
   &ScaleColor(
    Scale Alpha=&FloatFromCurve(FloatCurve=DataInterface() {propertyValues {Curve {keys [<at 0 0 RCIM_Cubic>,<at 0.25 1 RCIM_Cubic>,<at 1 0 RCIM_Cubic>]}}})
    Scale RGB=&VectorFromCurve(VectorCurve=DataInterface() {propertyValues {
       XCurve {keys [<at 0 15 RCIM_Cubic>,<at 1 1>]}
       YCurve {keys [<at 0 3 RCIM_Cubic>,<at 1 0.4798>]}
       ZCurve {keys [<at 0 1 RCIM_Cubic>,<at 1 0>]}
      }})
    ScaleA=true
    ScaleRGB=true
   )
   &SolveForcesAndVelocity(Clamp Velocity=false, Limit Acceleration=false)
  ]
  rendererClasses [
   &NiagaraMeshRendererProperties
   &NiagaraSpriteRendererProperties
   &NiagaraLightRendererProperties
  ]
  renderers [
   {rendererClass=&NiagaraMeshRendererProperties, rendererIndex=0}
   {rendererClass=&NiagaraSpriteRendererProperties, rendererIndex=1}
   {rendererClass=&NiagaraLightRendererProperties, rendererIndex=2}
  ]
 }
 TrailFlame {
  emitterSpawnScript [&FN_RibbonIDOffsetUtility()]
  emitterUpdateScript [
   &EmitterState(Life Cycle Mode=::NewEnumerator1, Loop Behavior=::NewEnumerator1, Loop Delay=0.02, Loop Duration=0.24, UseLoopDelay=true)
   &FN_SpawnAnimTrail()
  ]
  particleSpawnScript [
   &FN_InitializeAnimTrailBoneIndecies(Second Bone or Socket Index=0)
   &FN_InitializeAnimTrail(AnimTrailWidth=33, Color=<rgba 1 0.77 0.77 1>, Lifetime=0.2)
   &AddVelocity(Velocity=<1000 0 0>) {enabled=false}
  ]
  particleUpdateScript [
   &SolveForcesAndVelocity(Clamp Velocity=false, Limit Acceleration=false)
   &FN_UpdateAnimTrail() {module=FN_UpdateAnimTrail001}
   &ScaleColor(
    Scale Alpha=&Multiply_Float(
     A=&FloatFromCurve(
      FloatCurve=DataInterface() {propertyValues {Curve {keys [
          {arriveTangent=21.7156, interpMode=RCIM_Cubic, leaveTangent=21.7156, tangentMode=RCTM_User, time=0, value=0}
          {arriveTangent=-0.0116, interpMode=RCIM_Cubic, leaveTangent=-0.2338, tangentMode=RCTM_Break, time=0.1, value=1}
          {arriveTangent=-2.3964, interpMode=RCIM_Cubic, tangentMode=RCTM_Break, time=0.7, value=0}
         ]}}}
      Scale Curve=0.19
     )
     B=$System.SystemFade
    )
    Scale RGB=&MakeVectorFromLinearColorRGB(Color=&ColorFromCurve(ColorCurve=DataInterface() {propertyValues {
        AlphaCurve {keys []}
        BlueCurve {keys [<at 0.1863 0.9>,<at 0.9468 0.07>]}
        GreenCurve {keys [<at 0.1863 2.1>,<at 0.9468 0.225>]}
        RedCurve {keys [<at 0.1863 4.5>,<at 0.9468 1>]}
       }}))
    ScaleA=true
    ScaleRGB=true
   )
   &ScaleRibbonWidth(Ribbon Width Scale=&FloatFromCurve(FloatCurve=DataInterface() {propertyValues {Curve {keys [
         {arriveTangent=-0.8483, interpMode=RCIM_Cubic, leaveTangent=-0.8483, tangentMode=RCTM_User, time=0, value=1}
         {arriveTangent=-0.0978, interpMode=RCIM_Cubic, leaveTangent=-0.0978, tangentMode=RCTM_User, time=1, value=0.66}
        ]}}}))
   /SetVariables(
    Particles.ColorSmoke:LinearColor=&MultiplyLinearColor(
     Linear Color=&ColorFromCurve(ColorCurve=DataInterface() {propertyValues {
        AlphaCurve {keys [<at 0.0019 0>,<at 0.032 0.66 RCIM_Cubic>,<at 0.6744 0 RCIM_Cubic>]}
        BlueCurve {keys [<at 0.0345 0.136 RCIM_Cubic>,<at 0.3155 0.1>]}
        GreenCurve {keys [<at 0.0345 0.158 RCIM_Cubic>,<at 0.3155 0.1>]}
        RedCurve {keys [<at 0.0345 0.4 RCIM_Cubic>,<at 0.3155 0.1>]}
       }})
     Scale Factor=&MakeVector4FromVectorAndFloat("Vector (RGB)"=<1 1 1>, "Float (Alpha)"=$System.SystemFade)
    )
    Particles.RibbonWidthSmoke:NiagaraFloat=&Multiply_Float(
     A=50
     B=&FloatFromCurve(FloatCurve=DataInterface() {propertyValues {Curve {keys [<at 0 0.45>,<at 0.7 1>]}}})
    )
   ) {isSetParameters=true}
  ]
  rendererClasses [&NiagaraRibbonRendererProperties]
  renderers [
   {rendererClass=&NiagaraRibbonRendererProperties}
   {rendererClass=&NiagaraRibbonRendererProperties, rendererIndex=1}
  ]
 }
}
summary {
 systemName=NS_DieselPunk_Pickaxe_Swing
 emitters [
  {emitterName=Embers, rendererClasses [&NiagaraSpriteRendererProperties]}
  {emitterName=TrailFlame, rendererClasses [&NiagaraRibbonRendererProperties]}
  {
   emitterName=Thruster
   rendererClasses [
    &NiagaraMeshRendererProperties
    &NiagaraSpriteRendererProperties
    &NiagaraLightRendererProperties
   ]
  }
  {emitterName=Smoke, rendererClasses [&NiagaraSpriteRendererProperties]}
 ]
 userVariables [User.RibbonIDOffset {type=NiagaraInt32}]
}
